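# DriftTrail_Plexus is a Niagara vehicle cosmetic drift trail placed on skeletal mesh bones or sockets.
# Left and right WheelWrap emitters spin meshes tinted by User.PaintedColor, while Tech002/003 stream two-ribbon trails backward.
# Minimal rapidly re-bursts short-lived sprites with ribbons, and two UpwardMeshBurst emitters continuously spawn spinning meshes, one thrown up and back under gravity.
# Lifetime curves drive fades and mostly green-to-blue/purple color shifts; inherited source movement keeps trails following the car.
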NiagaraSystem /VehicleCosmetics/DriftTrail/DriftTrail_Plexus/FX/DriftTrail_Plexus
dependencies {
 usedDataInterfaces [
  &NiagaraDataInterfaceCurve
  &NiagaraDataInterfaceMeshRendererInfo
  &NiagaraDataInterfaceColorCurve
  &NiagaraDataInterfaceVectorCurve
  &NiagaraDataInterfaceSkeletalMesh
 ]
 usedDynamicInputs []
 usedModules [
  &self:SystemSpawnScript.SystemScriptSource.SystemScriptGraph.NiagaraNodeAssignment_0.SetVariables_#1
  &SystemState
 ]
}
emitters {
 Minimal {
  emitterUpdateScript [
   &EmitterState(
    Life Cycle Mode=::NewEnumerator1
    Loop Duration=&RandomRangeFloat(Minimum=0.01, Maximum=0.04, Evaluation Type=::NewEnumerator1)
    UseLoopDelay=false
   )
   &SpawnBurst_Instantaneous(Spawn Count=7, Use Spawn Probability=false, UseLoopCountLimit=false)
  ]
  particleSpawnScript [
   &PartitionParticles(Write Partition to Ribbon ID=::NewEnumerator0)
   &InitializeParticle(
    Lifetime=&RandomRangeFloat(Minimum=0.25, Maximum=0.3)
    Ribbon Width=3
    Ribbon Width Mode=::NewEnumerator1
    Sprite Size Mode=::NewEnumerator3
    Uniform Sprite Size=6
    UsePositionOffset=false
    Write Material Random=true
   )
   &SkeletalMeshLocation(
    Bone / Socket Index=&Add_Integer(A=$Particles.PartitionParticles.Partition, B=4)
    Bone / Socket Sampling Mode=::NewEnumerator5
    Interpolate Mesh Sampling=false
    MaskSpawnGroup=false
    Mesh Sampling Type=::NewEnumerator6
    Sampled Position Offset=<20 0 0>
    Skeletal Mesh=$System.SkeletalMesh
   )
   &ShapeLocation(Box Size=<0 30 35>, MaskSpawnGroup=false, Override Seed=false, Shape Primitive=::NewEnumerator2, Surface Only=false, UseOwnerScale=true)
   &AddVelocity(Velocity=&RandomRangeVector(Minimum=<-500 0 0>, Maximum=<-1000 0 0>))
   &ApplyInitialForces()
  ]
  particleUpdateScript [
   &ParticleState()
   &Drag(Drag=&RandomRangeFloat(Minimum=0.8, Maximum=1.2), Use Linear Drag=true, Use Rotational Drag=false) {enabled=false}
   &InheritSourceMovement(
    Applied Position Scale=&Multiply_VectorByFloat(
     Vector=<1 1 1>
     Float=&FloatFromCurve(FloatCurve=DataInterface() {propertyValues {Curve {keys [<at 0.3 1 RCIM_Cubic>, {arriveTangent=-0.3638, interpMode=RCIM_Cubic, leaveTangent=-0.3638, tangentMode=RCTM_User, time=1, value=0}]}}})
    )
    ApplyForce=false
    ApplyPosition=true
    ApplyPositionFalloff=false
   ) {module=InheritSourceMovement001}
   &SolveForcesAndVelocity(Clamp Velocity=false, Limit Acceleration=false)
   &ScaleColor(
    Scale Alpha=&FloatFromCurve(
     FloatCurve=DataInterface() {propertyValues {Curve {
        defaultValue=1
        keys [
         {arriveTangent=-0.0113, interpMode=RCIM_Cubic, leaveTangent=-0.0113, tangentMode=RCTM_User, time=0.25, value=0}
         {arriveTangent=0.0428, interpMode=RCIM_Cubic, leaveTangent=0.0428, tangentMode=RCTM_User, time=0.35, value=1}
         {arriveTangent=-0.3597, interpMode=RCIM_Cubic, leaveTangent=-0.3597, tangentMode=RCTM_User, time=0.5, value=0.9981}
         {arriveTangent=-0.1025, interpMode=RCIM_Cubic, leaveTangent=-0.1025, tangentMode=RCTM_User, time=1, value=0}
        ]
       }}}
     Scale Curve=0.4
    )
    Scale RGB=&MakeVectorFromLinearColorRGB(Color=&ColorFromCurve(ColorCurve=DataInterface() {propertyValues {
        AlphaCurve {keys [<at 0 1 RCIM_Cubic>]}
        BlueCurve {keys [<at 0.055 0>,<at 0.4543 0>,<at 0.5395 0.7225>,<at 0.6153 1>,<at 0.6989 1>,<at 0.8181 1>]}
        GreenCurve {keys [<at 0.055 5>,<at 0.4543 1>,<at 0.5395 1>,<at 0.6153 0.1905>,<at 0.6989 0.0298>,<at 0.8181 0>]}
        RedCurve {keys [<at 0.055 0>,<at 0.4543 0>,<at 0.5395 0>,<at 0.6153 0>,<at 0.6989 0.1994>,<at 0.8181 0.4745>]}
       }}))
    ScaleA=true
    ScaleRGB=true
   )
   /SetVariables(Particles.RibbonColor:LinearColor=&ColorFromCurve(ColorCurve=DataInterface() {propertyValues {
       AlphaCurve {keys [<at 0 1 RCIM_Cubic>,<at 1 0 RCIM_Cubic>]}
       BlueCurve {keys [
         <at 0.0044 0 RCIM_Cubic>
         <at 0.2632 11.8056>
         <at 0.4387 3>
         {arriveTangent=-5.7277, interpMode=RCIM_Cubic, leaveTangent=-5.7277, time=0.6549, value=1}
         <at 0.7879 1>
        ]}
       GreenCurve {keys [
         <at 0.0044 200 RCIM_Cubic>
         <at 0.2632 11.2104>
         <at 0.4387 0.185>
         {arriveTangent=-0.5298, interpMode=RCIM_Cubic, leaveTangent=-0.5298, time=0.6549, value=0.0347}
         <at 0.7879 0>
        ]}
       RedCurve {keys [
         <at 0.0044 0 RCIM_Cubic>
         <at 0.2632 0>
         <at 0.4387 0>
         {arriveTangent=2.1994, interpMode=RCIM_Cubic, leaveTangent=2.1994, time=0.6549, value=0.3449}
         <at 0.7879 0.768>
        ]}
      }})) {enabled=false, isSetParameters=true}
   &Color() {enabled=false}
  ]
  rendererClasses [&NiagaraSpriteRendererProperties, &NiagaraRibbonRendererProperties]
  renderers [
   {rendererClass=&NiagaraSpriteRendererProperties}
   {rendererClass=&NiagaraRibbonRendererProperties, rendererIndex=1}
  ]
 }
 NE_VC_WheelWrap_L {
  emitterUpdateScript [
   &EmitterState(
    Inactive Response=::NewEnumerator1
    Life Cycle Mode=::NewEnumerator1
    Loop Behavior=::NewEnumerator1
    Loop Duration Mode=::NewEnumerator1
    UseLoopDelay=false
   )
   &SpawnBurst_Instantaneous(Use Spawn Probability=false, UseLoopCountLimit=false)
   &VC_WheelWrapScale()
  ]
  particleSpawnScript [
   &VC_PaintedVariantsOverride() {enabled=false}
   &InitializeParticle(
    Color=$User.PaintedColor
    Mesh Scale=$Emitter.WheelFXScale
    Mesh Scale Mode=::NewEnumerator4
    Sprite Rotation Mode=::NewEnumerator2
    Sprite Size Mode=::NewEnumerator3
    UsePositionOffset=false
    Write Material Random=true
   )
   &PartitionParticles() {enabled=false}
   &SkeletalMeshLocation(
    Bone / Socket Sampling Mode=::NewEnumerator5
    Interpolate Mesh Sampling=false
    MaskSpawnGroup=false
    Mesh Sampling Type=::NewEnumerator6
    Orientation Sampling=::NewEnumerator1
    Sampled Position Offset Coordinate Space:ENiagaraSkelMeshTransforms=::NewEnumerator2
    Skeletal Mesh=$System.SkeletalMesh
   )
  ]
  particleUpdateScript [
   &ParticleState(Kill Particles When Lifetime Has Elapsed=false)
   &UpdateMeshOrientation(Rotation Coordinate Space=::NewEnumerator2, Rotation Rate=3, Rotation Vector=<0 1 0>)
   &ScaleColor(
    Scale Alpha=&FloatFromCurve(FloatCurve=DataInterface() {propertyValues {Curve {keys [
         {interpMode=RCIM_Cubic, tangentMode=RCTM_User, time=0, value=0}
         {interpMode=RCIM_Cubic, tangentMode=RCTM_User, time=0.5, value=1}
         {interpMode=RCIM_Cubic, tangentMode=RCTM_User, time=1, value=0}
        ]}}})
    Scale RGB=&VectorFromFloat(Value=$User.WheelIntensity)
    ScaleA=false
    ScaleRGB=true
   )
   &SolveForcesAndVelocity(Clamp Velocity=false, Limit Acceleration=false)
  ]
  rendererClasses [&NiagaraMeshRendererProperties]
  renderers [{rendererClass=&NiagaraMeshRendererProperties, rendererIndex=0}]
 }
 NE_VC_WheelWrap_R {
  emitterUpdateScript [
   &EmitterState(
    Inactive Response=::NewEnumerator1
    Life Cycle Mode=::NewEnumerator1
    Loop Behavior=::NewEnumerator1
    Loop Duration Mode=::NewEnumerator1
    UseLoopDelay=false
   )
   &SpawnBurst_Instantaneous(Use Spawn Probability=false, UseLoopCountLimit=false)
   &VC_WheelWrapScale()
  ]
  particleSpawnScript [
   &VC_PaintedVariantsOverride() {enabled=false}
   &InitializeParticle(
    Color=$User.PaintedColor
    Mesh Scale=$Emitter.WheelFXScale
    Mesh Scale Mode=::NewEnumerator4
    Sprite Rotation Mode=::NewEnumerator2
    Sprite Size Mode=::NewEnumerator3
    UsePositionOffset=false
    Write Material Random=true
   )
   &PartitionParticles() {enabled=false}
   &SkeletalMeshLocation(
    Bone / Socket Index=1
    Bone / Socket Sampling Mode=::NewEnumerator5
    Interpolate Mesh Sampling=false
    MaskSpawnGroup=false
    Mesh Sampling Type=::NewEnumerator6
    Orientation Sampling=::NewEnumerator1
    Sampled Position Offset Coordinate Space:ENiagaraSkelMeshTransforms=::NewEnumerator2
    Skeletal Mesh=$System.SkeletalMesh
   )
  ]
  particleUpdateScript [
   &ParticleState(Kill Particles When Lifetime Has Elapsed=false)
   &UpdateMeshOrientation(Rotation Coordinate Space=::NewEnumerator2, Rotation Rate=3, Rotation Vector=<0 1 0>)
   &ScaleColor(
    Scale Alpha=&FloatFromCurve(FloatCurve=DataInterface() {propertyValues {Curve {keys [
         {interpMode=RCIM_Cubic, tangentMode=RCTM_User, time=0, value=0}
         {interpMode=RCIM_Cubic, tangentMode=RCTM_User, time=0.5, value=1}
         {interpMode=RCIM_Cubic, tangentMode=RCTM_User, time=1, value=0}
        ]}}})
    Scale RGB=&VectorFromFloat(Value=$User.WheelIntensity)
    ScaleA=false
    ScaleRGB=true
   )
   &SolveForcesAndVelocity(Clamp Velocity=false, Limit Acceleration=false)
  ]
  rendererClasses [&NiagaraMeshRendererProperties]
  renderers [{rendererClass=&NiagaraMeshRendererProperties, rendererIndex=0}]
 }
 Tech002 {
  emitterUpdateScript [
   &EmitterState()
   &VC_SpawnRibbon_v1(Number of Ribbons=2, SpawnRate=60)
  ]
  particleSpawnScript [
   &InitializeParticle(Ribbon Width=16, Ribbon Width Mode=::NewEnumerator1, UsePositionOffset=false, Write Material Random=true) {enabled=false}
   &VC_PaintedVariantsOverride() {enabled=false}
   &VC_InitializeRibbon_v1(
    Color=<rgba 0 1 0.1394 1>
    Custom Offset=true
    Enable Skeletal Position=true
    Facing Axis:ENiagaraOrientationAxis=::ZAxis
    Lifetime=1.5
    Local Vector Offset=$User.offset
    Offset Amount=1
    Ribbon Width:NiagaraFloat=&Multiply_Float(A=1, B=$User.RESCALE)
   )
   &AddVelocity(Velocity=<-1000 0 0>) {module=AddVelocity001}
   &OffsetPosition(Position Offset=<0 0 -18>) {enabled=false}
  ]
  particleUpdateScript [
   &ParticleState()
   &ScaleColor(
    Scale Alpha=&FloatFromCurve(
     FloatCurve=DataInterface() {propertyValues {Curve {keys [
         {arriveTangent=4.4367, interpMode=RCIM_Cubic, leaveTangent=4.4367, tangentMode=RCTM_User, time=0, value=0.5}
         {arriveTangent=-0.0495, interpMode=RCIM_Cubic, leaveTangent=-0.0495, tangentMode=RCTM_User, time=0.1, value=1}
         {arriveTangent=-3.4203, interpMode=RCIM_Cubic, leaveTangent=-3.4203, tangentMode=RCTM_User, time=1, value=0}
        ]}}}
     Scale Curve=0.8
    )
    Scale RGB=&MakeVectorFromLinearColorRGB(Color=&ColorFromCurve(ColorCurve=DataInterface() {propertyValues {
        AlphaCurve {keys [<at 0.6102 1 RCIM_Cubic>,<at 1 0 RCIM_Cubic>]}
        BlueCurve {keys [<at 0.1961 0.5 RCIM_Cubic>,<at 0.3238 3>,<at 0.5299 5>,<at 0.758 4 RCIM_Cubic>]}
        GreenCurve {keys [<at 0.1961 2 RCIM_Cubic>,<at 0.3238 0.811>,<at 0.5299 0>,<at 0.758 0 RCIM_Cubic>]}
        RedCurve {keys [<at 0.1961 0.012 RCIM_Cubic>,<at 0.3238 0>,<at 0.5299 0>,<at 0.758 4 RCIM_Cubic>]}
       }}))
    ScaleA=true
    ScaleRGB=true
   )
   &DynamicMaterialParameters(
    Index 0 Param 1=20
    Index 0 Param 2=20
    Index 0 Param 3=$User.RibbonAddMult
    Index 0 Param 4=$User.RibbonBrightness
    Module.Index0Param0Write=true
    Module.Index0Param1Write=true
    Module.Index0Param2Write=true
    Module.Index0Param3Write=true
   )
   &InheritSourceMovement(
    Applied Position Scale=&Multiply_VectorByFloat(
     Vector=<1 1 1>
     Float=&FloatFromCurve(FloatCurve=DataInterface() {propertyValues {Curve {keys [<at 0.3 1 RCIM_Cubic>,<at 1 0 RCIM_Cubic>]}}})
    )
    ApplyForce=false
    ApplyPosition=true
    ApplyPositionFalloff=false
   )
   &SolveForcesAndVelocity(Clamp Velocity=false, Limit Acceleration=false)
   &VC_UpdateRibbon_v1(Enable Skeletal Position=true)
   &ScaleRibbonWidth(Ribbon Width Scale=&FloatFromCurve(FloatCurve=DataInterface() {propertyValues {Curve {
        defaultValue=1
        keys [
         {arriveTangent=-7.7111, interpMode=RCIM_Cubic, leaveTangent=-7.7111, tangentMode=RCTM_User, time=0, value=1}
         {arriveTangent=-0.0344, interpMode=RCIM_Cubic, leaveTangent=-0.0344, tangentMode=RCTM_User, time=0.2, value=0.375}
         {arriveTangent=0.0224, interpMode=RCIM_Cubic, leaveTangent=0.0224, tangentMode=RCTM_User, time=0.4, value=0.4}
        ]
       }}}))
  ]
  rendererClasses [&NiagaraRibbonRendererProperties]
  renderers [{rendererClass=&NiagaraRibbonRendererProperties}]
 }
 Tech003 {
  emitterUpdateScript [
   &EmitterState()
   &VC_SpawnRibbon_v1(Number of Ribbons=2, SpawnRate=60)
  ]
  particleSpawnScript [
   &InitializeParticle(Ribbon Width=16, Ribbon Width Mode=::NewEnumerator1, UsePositionOffset=false, Write Material Random=true) {enabled=false}
   &VC_PaintedVariantsOverride() {enabled=false}
   &VC_InitializeRibbon_v1(
    Color=<rgba 2 2 2 1>
    Custom Offset=true
    Enable Skeletal Position=true
    Facing Axis:ENiagaraOrientationAxis=::ZAxis
    Lifetime=1
    Local Vector Offset=$User.offset
    Offset Amount=1
    Ribbon Width:NiagaraFloat=&Multiply_Float(A=1, B=$User.RESCALE)
   )
   &AddVelocity(Velocity=<-1000 0 0>) {module=AddVelocity001}
   &OffsetPosition(Position Offset=<0 0 -18>) {enabled=false}
  ]
  particleUpdateScript [
   &ParticleState()
   &ScaleColor(
    Scale Alpha=&FloatFromCurve(FloatCurve=DataInterface() {propertyValues {Curve {keys [<at 0 1 RCIM_Cubic>, {arriveTangent=-2.3612, interpMode=RCIM_Cubic, leaveTangent=-2.3612, tangentMode=RCTM_User, time=1, value=0}]}}})
    Scale RGB=&MakeVectorFromLinearColorRGB(Color=&ColorFromCurve(ColorCurve=DataInterface() {propertyValues {
        AlphaCurve {keys [<at 0 0>,<at 0.0811 1 RCIM_Cubic>,<at 1 0 RCIM_Cubic>]}
        BlueCurve {keys [
          <at 0.0047 0.5>
          {arriveTangent=4.5787, interpMode=RCIM_Cubic, leaveTangent=4.5787, time=0.1471, value=0.2}
          {arriveTangent=5.6137, interpMode=RCIM_Cubic, leaveTangent=5.6137, time=0.2173, value=1.4735}
          <at 0.3385 1.2742>
          <at 0.4728 1>
          <at 0.6941 1>
         ]}
        GreenCurve {keys [
          <at 0.0047 2>
          {arriveTangent=-4.7033, interpMode=RCIM_Cubic, leaveTangent=-4.7033, time=0.1471, value=1}
          {arriveTangent=-1.6872, interpMode=RCIM_Cubic, leaveTangent=-1.6872, time=0.2173, value=1}
          <at 0.3385 0.6772>
          <at 0.4728 0>
          <at 0.6941 0.0453>
         ]}
        RedCurve {keys [<at 0.0047 0>,<at 0.1471 0 RCIM_Cubic>,<at 0.2173 0 RCIM_Cubic>,<at 0.3385 0>,<at 0.4728 0.0849>,<at 0.6941 0.4878>]}
       }}))
    ScaleA=true
    ScaleRGB=true
   )
   &DynamicMaterialParameters(
    Index 0 Param 1=20
    Index 0 Param 2=20
    Index 0 Param 3=$User.RibbonAddMult
    Index 0 Param 4=$User.RibbonBrightness
    Module.Index0Param0Write=true
    Module.Index0Param1Write=true
    Module.Index0Param2Write=true
    Module.Index0Param3Write=true
   )
   &InheritSourceMovement(
    Applied Position Scale=&Multiply_VectorByFloat(
     Vector=<1 1 1>
     Float=&FloatFromCurve(FloatCurve=DataInterface() {propertyValues {Curve {keys [<at 0.3 1 RCIM_Cubic>,<at 1 0 RCIM_Cubic>]}}})
    )
    ApplyForce=false
    ApplyPosition=true
    ApplyPositionFalloff=false
   )
   &SolveForcesAndVelocity(Clamp Velocity=false, Limit Acceleration=false)
   &VC_UpdateRibbon_v1(Enable Skeletal Position=true)
   &ScaleRibbonWidth(Ribbon Width Scale=&FloatFromCurve(FloatCurve=DataInterface() {propertyValues {Curve {
        defaultValue=1
        keys [
         {arriveTangent=-7.7111, interpMode=RCIM_Cubic, leaveTangent=-7.7111, tangentMode=RCTM_User, time=0, value=1}
         {arriveTangent=-0.0344, interpMode=RCIM_Cubic, leaveTangent=-0.0344, tangentMode=RCTM_User, time=0.2, value=0.3765}
         {arriveTangent=0.0224, interpMode=RCIM_Cubic, leaveTangent=0.0224, tangentMode=RCTM_User, time=1, value=0.4}
        ]
       }}}))
  ]
  rendererClasses [&NiagaraRibbonRendererProperties]
  renderers [{rendererClass=&NiagaraRibbonRendererProperties}]
 }
 UpwardMeshBurst003 {
  emitterUpdateScript [
   &EmitterState()
   &SpawnRate(SpawnRate=200, Use Spawn Probability=false)
   &VC_WheelWrapScale()
   &SpawnBurst_Instantaneous(Spawn Count=30, Use Spawn Probability=false, UseLoopCountLimit=false) {enabled=false}
  ]
  particleSpawnScript [
   &InitializeParticle(
    Lifetime Max=0.7
    Lifetime Min=0.15
    Lifetime Mode=::NewEnumerator1
    Mesh Renderer Array Visibility Mode=::NewEnumerator1
    Mesh Renderer Info=DataInterface() {propertyValues {MeshRenderer=&self:UpwardMeshBurst003_0.NiagaraMeshRendererProperties_0}}
    Mesh Scale=$Emitter.WheelFXScale
    Mesh Scale Mode=::NewEnumerator4
    Position Offset=<0 0 -120>
    UsePositionOffset=true
    Write Material Random=true
   )
   &PartitionParticles()
   &SkeletalMeshLocation(
    Bone / Socket Index=&Add_Integer(A=$Particles.PartitionParticles.Partition, B=4)
    Bone / Socket Sampling Mode=::NewEnumerator5
    Interpolate Mesh Sampling=false
    MaskSpawnGroup=false
    Mesh Sampling Type=::NewEnumerator6
    Skeletal Mesh=$System.SkeletalMesh
   )
   &ShapeLocation(Box Size=<1 32 1>, MaskSpawnGroup=false, Override Seed=false, Shape Primitive=::NewEnumerator2, Surface Only=false, UseOwnerScale=true)
   &AddVelocity(Velocity=&RandomRangeVector(Minimum=<0 0 0>, Maximum=<0 0 0>)) {enabled=false}
   &InitialMeshOrientation(
    Mesh Axis To Orient=<0 0 0>
    Mesh Orientation Mode=::NewEnumerator2
    Orientation Vector=<0 0 0>
    Rotation=&RandomRangeVector(Minimum=<0 0 0>, Maximum=<0 1 0>)
    Use Rotation Vector=true
   )
  ]
  particleUpdateScript [
   &ParticleState()
   &GravityForce() {enabled=false}
   &UpdateMeshOrientation(
    Rotation Rate=&FloatFromCurve(
     FloatCurve=DataInterface() {propertyValues {Curve {defaultValue=1, keys [{tangentMode=RCTM_User, time=0, value=1}, {tangentMode=RCTM_User, time=1, value=1}]}}}
     Scale Curve=3
    )
    Rotation Vector=&RandomRangeVector(Minimum=<0 1 0>, Maximum=<0 1.5 0>)
   )
   &Drag(Drag=4, Use Linear Drag=true, Use Rotational Drag=true) {enabled=false}
   &ScaleMeshSize(Scale Factor=&VectorFromCurve(
     VectorCurve=DataInterface() {propertyValues {XCurve {keys [<at 0 1>,<at 1 1>]}, YCurve {keys [<at 0 1>,<at 1 1>]}, ZCurve {keys [<at 0 1>,<at 1 1>]}}}
     Scale Curve=&RandomRangeVector(Minimum=<0.9 0.1 0.9>, Maximum=<0.9 0.6 0.9>)
    ))
   &ScaleColor(
    Scale Alpha=&FloatFromCurve(FloatCurve=DataInterface() {propertyValues {Curve {keys [<at 0.5 1 RCIM_Cubic>,<at 1.0011 0.002 RCIM_Cubic>]}}}, Scale Curve=0.5)
    Scale RGB=&MakeVectorFromLinearColorRGB(Color=&ColorFromCurve(ColorCurve=DataInterface() {propertyValues {
        AlphaCurve {keys [<at 0 1 RCIM_Cubic>,<at 1 0 RCIM_Cubic>]}
        BlueCurve {keys [<at 0.0532 0.1101>,<at 0.1352 0.2957>,<at 0.3366 1>,<at 0.4952 1>]}
        GreenCurve {keys [<at 0.0532 2>,<at 0.1352 2>,<at 0.3366 0.317>,<at 0.4952 0>]}
        RedCurve {keys [<at 0.0532 0>,<at 0.1352 0>,<at 0.3366 0>,<at 0.4952 1>]}
       }}))
    ScaleA=true
    ScaleRGB=true
   )
   &AddRotationalVelocity() {enabled=false}
   &MeshRotationForce(Rotation=<30000 100 100>) {enabled=false}
   &DynamicMaterialParameters(
    Index 0 Param 1=&FloatFromCurve(FloatCurve=DataInterface() {propertyValues {Curve {keys [<at -0.0021 1>,<at 1 0>]}}})
    Index 0 Param 2=&FloatFromCurve(FloatCurve=DataInterface() {propertyValues {Curve {
        defaultValue=1
        keys [
         {arriveTangent=-0.1035, interpMode=RCIM_Cubic, leaveTangent=-0.1035, tangentMode=RCTM_User, time=0.0038, value=1}
         {arriveTangent=-0.5483, interpMode=RCIM_Cubic, leaveTangent=-0.5483, tangentMode=RCTM_User, time=1, value=0}
        ]
       }}})
    Index 0 Param 3=&FloatFromCurve(
     FloatCurve=DataInterface() {propertyValues {Curve {defaultValue=1, keys [{tangentMode=RCTM_User, time=0, value=2}, <at 0.2878 0.8849>, {tangentMode=RCTM_User, time=1, value=0.6}]}}}
     Scale Curve=&RandomRangeFloat(Minimum=0.2, Maximum=1)
    )
    Index 0 Param 4=0.02
    Index 1 Param 2=&FloatFromCurve(FloatCurve=DataInterface() {propertyValues {Curve {keys [<at 0 1>,<at 1 0>]}}})
    Module.Index0Float0Write=true
    Module.Index0Float1Write=true
    Module.Index0Float2Write=true
    Module.Index0Float3Write=true
    Module.Index1Float0Write=true
    Module.Index1Float1Write=true
    Module.Index1Float2Write=true
    Module.Index1Float3Write=true
    Write Parameter Index 1=::NewEnumerator1
   ) {module=DynamicMaterialParameters001}
   &Collision(
    Age Colliding Particles=false
    Friction=0.1
    KillOnCollision=false
    Module.Control Rest State Via Surface Normal=false
    Module.EnableMaxCollisionCount=false
    Module.Randomize Collision Normal Vector=false
    Restitution=3
   ) {enabled=false}
   &GravityForce(Gravity=<0 0 -2000>) {enabled=false, module=GravityForce001}
   &InheritSourceMovement(
    Applied Position Scale=&Multiply_VectorByFloat(
     Vector=<1 1 1>
     Float=&FloatFromCurve(FloatCurve=DataInterface() {propertyValues {Curve {keys [<at 0 1 RCIM_Cubic>,<at 0.9973 0 RCIM_Cubic>]}}})
    )
    ApplyForce=false
    ApplyPosition=true
    ApplyPositionFalloff=false
   ) {enabled=false}
   &VortexForce(Origin Pull Amount=500, UseOriginOffset=false, Vortex Axis=<0 1 0>, Vortex Force Amount=10000) {enabled=false}
   &SolveForcesAndVelocity(Clamp Velocity=false, Limit Acceleration=false)
  ]
  rendererClasses [&NiagaraMeshRendererProperties]
  renderers [{rendererClass=&NiagaraMeshRendererProperties, rendererIndex=0}]
 }
 UpwardMeshBurst004 {
  emitterUpdateScript [
   &EmitterState()
   &SpawnRate(SpawnRate=200, Use Spawn Probability=false)
   &SpawnBurst_Instantaneous(Spawn Count=30, Use Spawn Probability=false, UseLoopCountLimit=false) {enabled=false}
  ]
  particleSpawnScript [
   &InitializeParticle(
    AdjustAlpha=false
    AdjustHue=true
    AdjustSaturation=false
    AdjustValue=false
    Color=<rgba 0 2 0 1>
    Color Mode=::NewEnumerator3
    Hue Shift Range:Vector2f=<0 1>
    Lifetime Max=0.3
    Lifetime Min=0.1
    Lifetime Mode=::NewEnumerator1
    Mesh Renderer Array Visibility Mode=::NewEnumerator3
    Mesh Renderer Info=DataInterface() {propertyValues {MeshRenderer=&self:UpwardMeshBurst004_0.NiagaraMeshRendererProperties_0}}
    Mesh Scale Mode=::NewEnumerator1
    Mesh Uniform Scale Max=0.08
    Mesh Uniform Scale Min=0.05
    Position Offset=<0 0 -120>
    UsePositionOffset=true
    Write Material Random=true
   )
   &PartitionParticles()
   &SkeletalMeshLocation(
    Bone / Socket Index=&Add_Integer(A=$Particles.PartitionParticles.Partition, B=4)
    Bone / Socket Sampling Mode=::NewEnumerator5
    Interpolate Mesh Sampling=false
    MaskSpawnGroup=false
    Mesh Sampling Type=::NewEnumerator6
    Sampled Position Offset=<-30 0 -40>
    Skeletal Mesh=$System.SkeletalMesh
   )
   &ShapeLocation(Box Size=<0 50 0>, MaskSpawnGroup=false, Override Seed=false, Shape Primitive=::NewEnumerator2, Surface Only=false, UseOwnerScale=true)
   &AddVelocity(Velocity=&RandomRangeVector(Minimum=<-300 50 300>, Maximum=<-800 -50 900>))
   &InitialMeshOrientation(Use Rotation Vector=false)
  ]
  particleUpdateScript [
   &ParticleState()
   &GravityForce() {enabled=false}
   &UpdateMeshOrientation(
    Rotation Rate=&FloatFromCurve(
     FloatCurve=DataInterface() {propertyValues {Curve {
        defaultValue=2
        keys [
         {arriveTangent=-0.8515, interpMode=RCIM_Cubic, leaveTangent=-0.8515, tangentMode=RCTM_User, time=0, value=2}
         {arriveTangent=-0.0796, interpMode=RCIM_Cubic, leaveTangent=-0.0796, tangentMode=RCTM_User, time=1, value=0}
        ]
       }}}
     Scale Curve=2
    )
    Rotation Vector=&RandomRangeVector(Minimum=<-0.5 -0.5 -0.5>, Maximum=<0.5 0.5 0.5>)
   )
   &Drag(Drag=4, Use Linear Drag=true, Use Rotational Drag=true)
   &ScaleMeshSize(Scale Factor=&VectorFromFloat(Value=&FloatFromCurve(FloatCurve=DataInterface() {propertyValues {Curve {keys [
          {arriveTangent=-0.1097, interpMode=RCIM_Cubic, leaveTangent=-0.1097, tangentMode=RCTM_User, time=0.0008, value=1}
          {arriveTangent=-1.3012, tangentMode=RCTM_User, time=1, value=0.2}
         ]}}})))
   &ScaleColor(
    Scale Alpha=&FloatFromCurve(FloatCurve=DataInterface() {propertyValues {Curve {keys [<at 0.5 0.5 RCIM_Cubic>, {arriveTangent=-0.3704, interpMode=RCIM_Cubic, leaveTangent=-0.3704, tangentMode=RCTM_User, time=1.0011, value=0.1}]}}})
    Scale RGB=&MakeVectorFromLinearColorRGB(Color=&ColorFromCurve(ColorCurve=DataInterface() {propertyValues {
        AlphaCurve {keys [<at 0 1 RCIM_Cubic>]}
        BlueCurve {keys [<at 0.0042 1>]}
        GreenCurve {keys [<at 0.0042 2>]}
        RedCurve {keys [<at 0.0042 0>]}
       }}))
    ScaleA=true
    ScaleRGB=true
   )
   &AddRotationalVelocity() {enabled=false}
   &MeshRotationForce(Rotation=<30000 100 100>) {enabled=false}
   &DynamicMaterialParameters(
    Index 0 Param 1=&FloatFromCurve(FloatCurve=DataInterface() {propertyValues {Curve {keys [<at 0 1>,<at 1 0>]}}})
    Index 0 Param 2=&FloatFromCurve(FloatCurve=DataInterface() {propertyValues {Curve {keys [<at 0 1>,<at 0.7 1>,<at 1 0>]}}})
    Index 0 Param 3=&FloatFromCurve(
     FloatCurve=DataInterface() {propertyValues {Curve {
        defaultValue=1
        keys [
         {arriveTangent=-1.1481, interpMode=RCIM_Cubic, leaveTangent=-1.1481, tangentMode=RCTM_User, time=0, value=1}
         {arriveTangent=-0.2571, interpMode=RCIM_Cubic, leaveTangent=-0.2571, tangentMode=RCTM_User, time=1, value=0.5}
        ]
       }}}
     Scale Curve=5
    )
    Index 0 Param 4=0.1
    Index 1 Param 1=&FloatFromCurve(
     FloatCurve=DataInterface() {propertyValues {Curve {
        defaultValue=1
        keys [
         {arriveTangent=-2.9181, interpMode=RCIM_Cubic, leaveTangent=-2.9181, tangentMode=RCTM_User, time=0, value=1}
         {arriveTangent=0.0224, interpMode=RCIM_Cubic, leaveTangent=0.0224, tangentMode=RCTM_User, time=1, value=0}
        ]
       }}}
     Scale Curve=0.6
    )
    Index 1 Param 2=&FloatFromCurve(FloatCurve=DataInterface() {propertyValues {Curve {keys [<at 0 1>,<at 1 0.1>]}}}, Scale Curve=40)
    Module.Index0Float0Write=true
    Module.Index0Float1Write=true
    Module.Index0Float2Write=true
    Module.Index0Float3Write=true
    Module.Index1Float0Write=true
    Module.Index1Float1Write=true
    Module.Index1Float2Write=true
    Module.Index1Float3Write=true
    Write Parameter Index 1=::NewEnumerator1
   ) {module=DynamicMaterialParameters001}
   &Collision(
    Age Colliding Particles=false
    Friction=0.1
    KillOnCollision=false
    Module.Control Rest State Via Surface Normal=false
    Module.EnableMaxCollisionCount=false
    Module.Randomize Collision Normal Vector=false
    Restitution=3
   ) {enabled=false}
   &GravityForce(Gravity=<0 0 -2000>) {module=GravityForce001}
   &InheritSourceMovement(
    Applied Position Scale=&Multiply_VectorByFloat(
     Vector=<1 1 1>
     Float=&FloatFromCurve(FloatCurve=DataInterface() {propertyValues {Curve {keys [
          <at 0.3 0.5 RCIM_Cubic>
          {arriveTangent=-0.8392, interpMode=RCIM_Cubic, leaveTangent=-0.8392, tangentMode=RCTM_User, time=1.0015, value=0.2647}
         ]}}})
    )
    ApplyForce=false
    ApplyPosition=true
    ApplyPositionFalloff=false
   )
   &VortexForce(Origin Pull Amount=500, UseOriginOffset=false, Vortex Axis=<0 1 0>, Vortex Force Amount=10000) {enabled=false}
   &SolveForcesAndVelocity(Clamp Velocity=false, Limit Acceleration=false)
  ]
  rendererClasses [&NiagaraMeshRendererProperties]
  renderers [{rendererClass=&NiagaraMeshRendererProperties, rendererIndex=0}]
 }
}
summary {
 systemName=DriftTrail_Plexus
 emitters [
  {emitterName=NE_VC_WheelWrap_L, rendererClasses [&NiagaraMeshRendererProperties]}
  {emitterName=NE_VC_WheelWrap_R, rendererClasses [&NiagaraMeshRendererProperties]}
  {emitterName=UpwardMeshBurst004, rendererClasses [&NiagaraMeshRendererProperties]}
  {emitterName=Tech002, rendererClasses [&NiagaraRibbonRendererProperties]}
  {emitterName=Tech003, rendererClasses [&NiagaraRibbonRendererProperties]}
  {emitterName=Minimal, rendererClasses [&NiagaraSpriteRendererProperties, &NiagaraRibbonRendererProperties]}
  {emitterName=UpwardMeshBurst003, rendererClasses [&NiagaraMeshRendererProperties]}
 ]
 userVariables [
  User.bIsFrontendPreview {type=NiagaraBool}
  User.bIsPainted {type=NiagaraBool}
  User.CodeBrightness {type=NiagaraFloat, defaultValue=10, description=""}
  User.ColorVariantID {type=NiagaraInt32}
  User.float {type=NiagaraFloat, defaultValue=0, description=""}
  User.NumbersColor {type=LinearColor, description="", defaultValue {struct=LinearColor, value=<rgba 0.0244 0.0762 1 1>}}
  User.offset {type=Vector3f, description="offset from wheels(set through cosmetic template)", defaultValue {struct=Vector3f, value=<-12 0 0>}}
  User.PaintedColor {type=LinearColor, description="", defaultValue {struct=LinearColor, value=<rgba 0 1 0.1 1>}}
  User.PaintedScalar {type=NiagaraFloat, defaultValue=1, description=""}
  User.RESCALE {type=NiagaraFloat, defaultValue=113, description=""}
  User.RibbonAddMult {type=NiagaraFloat, defaultValue=0.1, description=""}
  User.RibbonBrightness {type=NiagaraFloat, defaultValue=8, description=""}
  User.WheelIntensity {type=NiagaraFloat, defaultValue=1, description=""}
  User.WheelSize {type=NiagaraFloat, defaultValue=0, description=""}
  User.ZeroOneBrightness {type=NiagaraFloat, defaultValue=20, description=""}
 ]
}
